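; PCG sample graph that spawns copies of a Houdini-exported circular point pattern, imported from a datatable, at points sampled across the landscape.
; Surface Sampler output feeds Copy Points as targets; the copies pass through a Density Filter, then project back onto the landscape.
; CreateParamData nodes override sampling density, pattern scale and filtering, though the LowerBound parameter connects to the UpperBound pin.
; Debug nodes display sampled and projected points, with no mesh spawning.
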(graph SpawnPointsFromPointsGraph
  (bind GetLandscapeData_1 (Get%20Landscape%20Data :ActorSelector (struct :ActorFilter AllWorldActors :bMustOverlapSelf true :ActorSelection ByClass :ActorSelectionClass /Script/Landscape.LandscapeProxy :ActorReferenceSelector @LastCreated :bSelectMultiple true) :bComponentsMustOverlapSelf false :SamplingProperties (struct)))
  (bind SurfaceSamplerNode_0 (Surface%20Sampler :PointExtents (1e+02 1e+02 1e+02) :PointsPerSquaredMeter 0.05 :PointSteepness 0.5)) ; Sample Landscape 
  (bind BlueprintNode_1 (Execute%20Blueprint :BlueprintElementInstance (class:/PCG/BP_Elements/ImportDataTable.ImportDataTable_C :Scale 1.5e+02) :BlueprintElementType /PCG/BP_Elements/ImportDataTable.ImportDataTable_C)) ; Import a circular point pattern exported from Houdini and stored in a datatable
  (bind CopyPoints_5 (Copy%20Points))
  (bind DensityFilter_6 (Density%20Filter :LowerBound 0.0))
  (bind GetLandscapeData_4 (Get%20Landscape%20Data :ActorSelector (struct :ActorFilter AllWorldActors :bMustOverlapSelf true :ActorSelection ByClass :ActorSelectionClass /Script/Landscape.LandscapeProxy :ActorReferenceSelector @LastCreated :bSelectMultiple true) :bComponentsMustOverlapSelf false :SamplingProperties (struct)))
  (bind Projection_2 (Projection))
  (bind Debug_0 (Debug))
  (bind Debug_7 (Debug))
  ; Parameters, to override node properties by attribute name
  (bind BlueprintNode_0 (Execute%20Blueprint :BlueprintElementInstance (class:/PCG/BP_Elements/Deprecated/CreateParamData.CreateParamData_C :AttributeName PointsPerSquaredMeter :Value (0.03 0.0 0.0 0.0)) :BlueprintElementType /PCG/BP_Elements/Deprecated/CreateParamData.CreateParamData_C))
  (bind BlueprintNode_7 (Execute%20Blueprint :BlueprintElementInstance (class:/PCG/BP_Elements/Deprecated/CreateParamData.CreateParamData_C :AttributeName Scale :Value (1.5e+02 0.0 0.0 0.0)) :BlueprintElementType /PCG/BP_Elements/Deprecated/CreateParamData.CreateParamData_C))
  (bind BlueprintNode_9 (Execute%20Blueprint :BlueprintElementInstance (class:/PCG/BP_Elements/Deprecated/CreateParamData.CreateParamData_C :AttributeName LowerBound :Value (0.2 0.0 0.0 0.0)) :BlueprintElementType /PCG/BP_Elements/Deprecated/CreateParamData.CreateParamData_C))
  (connect BlueprintNode_7 Out BlueprintNode_1 Overrides)
  (connect BlueprintNode_1 Out CopyPoints_5 Source)
  (connect SurfaceSamplerNode_0 Out CopyPoints_5 Target)
  (connect Projection_2 Out Debug_0 In)
  (connect SurfaceSamplerNode_0 Out Debug_7 In)
  (connect CopyPoints_5 Out DensityFilter_6 In)
  (connect BlueprintNode_9 Out DensityFilter_6 UpperBound)
  (connect DensityFilter_6 Out Projection_2 In)
  (connect GetLandscapeData_4 Out Projection_2 Projection%20Target)
  (connect BlueprintNode_0 Out SurfaceSamplerNode_0 PointsPerSquaredMeter)
  (connect GetLandscapeData_1 Out SurfaceSamplerNode_0 Surface))
